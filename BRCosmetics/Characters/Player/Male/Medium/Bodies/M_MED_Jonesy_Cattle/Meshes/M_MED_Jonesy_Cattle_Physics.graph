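# Physics asset for the M_MED_Jonesy_Cattle player body, simulating two dangling grenades with fuses on the character.
# Capsule bodies cover spine_04, the right arm chain and each grenade's base, body and fuse bones.
# Each grenade base is nearly locked to spine_04, the grenade swings within tight, asymmetric limits, and its fuse hangs off it more loosely.
# The right arm chain joins spine_04 with wider limits, likely serving as collision for the grenades.
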PhysicsAsset /BRCosmetics/Characters/Player/Male/Medium/Bodies/M_MED_Jonesy_Cattle/Meshes/M_MED_Jonesy_Cattle_Physics
bodies {
 dyn_GrenadeFuse_01 [{shapeType=Capsule, length=1.6686, radius=1.6679, center=<-0.0079 0.0364 1.3841>, rotation=<rot -1.8546 -90 -0.7154>}]
 dyn_GrenadeFuse_02 [{shapeType=Capsule, length=1.6686, radius=1.6678, center=<0.0108 -0.0001 1.3845>, rotation=<rot 0.3633 -90 -2.8449>}]
 dyn_Grenade_01 [{shapeType=Capsule, length=8.4978, radius=2.874, center=<-0.2719 3.3358 -0.8182>, rotation=<rot -0.1936 -90 12.8495>}]
 dyn_Grenade_01_base [{shapeType=Capsule, length=0.505, radius=0.505, rotation=<rot 0 -90 0>}]
 dyn_Grenade_02 [{shapeType=Capsule, length=8.5158, radius=2.9792, center=<-0.6519 3.4343 -0.7344>, rotation=<rot -5.2258 -90 11.2063>}]
 dyn_Grenade_02_base [{shapeType=Capsule, length=0.505, radius=0.505, rotation=<rot 0 -90 0>}]
 hand_r [{shapeType=Capsule, length=10.0148, radius=5.8044, center=<-7.3725 2.3793 -2.6151>, rotation=<rot 0 86.6924 -87.8122>}]
 lowerarm_r [{shapeType=Capsule, length=11.1728, radius=4.354, center=<-6.366 -0.9368 -0.5881>, rotation=<rot 0.0874 85.0348 -89.1592>}]
 spine_04 [{shapeType=Capsule, length=9.9607, radius=12.6391, center=<1.4424 1.8391 -0.0173>, rotation=<rot 0 -5.3076 -178.8474>}]
 upperarm_r [{shapeType=Capsule, length=18.5227, radius=4.3266, center=<-11.737 0.1624 0.9015>, rotation=<rot 0 -88.1766 -88.2413>}]
}
constraints [
 {bone1Name=upperarm_r, bone2Name=spine_04, swing1LimitDegrees=45, swing2LimitDegrees=45, twistLimitDegrees=45}
 {bone1Name=lowerarm_r, bone2Name=upperarm_r, swing1LimitDegrees=45, swing2LimitDegrees=45, twistLimitDegrees=45}
 {bone1Name=hand_r, bone2Name=lowerarm_r, swing1LimitDegrees=45, swing2LimitDegrees=45, twistLimitDegrees=45}
 {bone1Name=dyn_Grenade_01_base, bone2Name=spine_04, swing1LimitDegrees=0.1, swing2LimitDegrees=0.1, twistLimitDegrees=0.1}
 {bone1Name=dyn_Grenade_01, bone2Name=dyn_Grenade_01_base, swing1LimitDegrees=1, swing2LimitDegrees=5, twistLimitDegrees=1}
 {bone1Name=dyn_Grenade_02_base, bone2Name=spine_04, swing1LimitDegrees=0.1, swing2LimitDegrees=0.1, twistLimitDegrees=0.1}
 {bone1Name=dyn_Grenade_02, bone2Name=dyn_Grenade_02_base, swing1LimitDegrees=1, swing2LimitDegrees=10, twistLimitDegrees=1}
 {bone1Name=dyn_GrenadeFuse_01, bone2Name=dyn_Grenade_01, swing1LimitDegrees=15, swing2LimitDegrees=15, twistLimitDegrees=15}
 {bone1Name=dyn_GrenadeFuse_02, bone2Name=dyn_Grenade_02, swing1LimitDegrees=25, swing2LimitDegrees=25, twistLimitDegrees=25}
]
