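# Animation Blueprint for the Jonesy Cattle cosmetic body, adding corrective and secondary motion to the pose copied from its parent mesh.
# Constraints and additive bone transforms adjust grenade and thigh bones, with thigh corrections scaled by leg rotation alphas computed each update.
# The cached result blends into extra grenade offsets while targeting, then feeds a pelvis-space RigidBody simulation.
# Shoulder alphas are also computed but drive nothing in the AnimGraph.
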AnimBlueprint /BRCosmetics/Characters/Player/Male/Medium/Bodies/M_MED_Jonesy_Cattle/Meshes/M_MED_Jonesy_Cattle_AnimBP
target_skeleton=&/M_MED_Jonesy_Cattle_Skeleton
graphs [
 AnimGraph {nodes [
   n1 = "Transform (Modify) Bone - Bone: dyn_Grenade_01_base"(ComponentPose=Local To Component(LocalPose=Use cached pose 'AnimCache'()), Alpha=1, Translation="-2,-1,0", Rotation="10,0,0", Scale="1,1,1") {
    rotationMode=BMM_Additive
    rotationSpace=BCS_BoneSpace
    scaleMode=BMM_Additive
    scaleSpace=BCS_BoneSpace
    translationMode=BMM_Additive
    translationSpace=BCS_BoneSpace
    alphaBoolBlend {blendOption=Linear}
    boneToModify {boneName=dyn_Grenade_01_base}
    rotation {pitch=10}
    translation=<-2 -1>
   }
   n2 = "Transform (Modify) Bone - Bone: dyn_Grenade_02_base"(ComponentPose=@n1, Alpha=1, Translation="-2,-1,0", Rotation="10,0,0", Scale="1,1,1") {
    rotationMode=BMM_Additive
    rotationSpace=BCS_BoneSpace
    scaleMode=BMM_Additive
    scaleSpace=BCS_BoneSpace
    translationMode=BMM_Additive
    translationSpace=BCS_BoneSpace
    alphaBoolBlend {blendOption=Linear}
    boneToModify {boneName=dyn_Grenade_02_base}
    rotation {pitch=10}
    translation=<-2 -1>
   }
   n3 = Component To Local(ComponentPose=@n2)
   n4 = Use cached pose 'AnimCache'()
   n5 = Get bIsTargeting()
   n6 = RigidBody(ComponentPose=Local To Component(LocalPose=Two Way Blend!(A=@n4, B=@n3, bAlphaBoolEnabled:bool=@n5) -> (Pose) {blendNode {alphaInputType=Bool, alphaBoolBlend {blendInTime=0.05, blendOption=Cubic, blendOutTime=0.1}}})) {
    bClampLinearTranslationLimitToRefPose=true
    bOverrideWorldGravity=true
    overridePhysicsAsset=&/M_MED_Jonesy_Cattle_Physics
    simulationSpace=BaseBoneSpace
    alphaBoolBlend {blendOption=Linear}
    baseBoneRef {boneName=pelvis}
    componentAppliedLinearAccClamp=<1000 1000 1000>
    componentLinearAccScale=<0.35 0.45 0.45>
    componentLinearVelScale=<0.3 0.3 0.3>
    overrideWorldGravity {z=-980}
   }
   Output Pose(Result=Component To Local(ComponentPose=@n6)) {layerGroup=None}
   n7 = "Transform (Modify) Bone - Bone: dyn_Grenade_01_base"(
    ComponentPose=Local To Component(LocalPose=Copy Pose From Mesh() {bUseAttachedParent=true})
    Alpha=1
    Translation="0,0,0"
    Rotation="5,0,0"
    Scale="1,1,1"
   ) {
    rotationMode=BMM_Additive
    rotationSpace=BCS_BoneSpace
    scaleMode=BMM_Additive
    scaleSpace=BCS_BoneSpace
    translationSpace=BCS_BoneSpace
    alphaBoolBlend {blendOption=Linear}
    boneToModify {boneName=dyn_Grenade_01_base}
    rotation {pitch=5}
   }
   n8 = "Transform (Modify) Bone - Bone: dyn_Grenade_02_base"(ComponentPose=@n7, Alpha=1, Translation="0,0,0", Rotation="5,0,0", Scale="1,1,1") {
    rotationMode=BMM_Additive
    rotationSpace=BCS_BoneSpace
    scaleMode=BMM_Additive
    scaleSpace=BCS_BoneSpace
    translationSpace=BCS_BoneSpace
    alphaBoolBlend {blendOption=Linear}
    boneToModify {boneName=dyn_Grenade_02_base}
    rotation {pitch=5}
   }
   n9 = "Constraint - thigh_l_fix_a (PP)"!(ComponentPose=@n8, Alpha=1, ConstraintWeights_0=0.4, ConstraintWeights_1:float=0.6) -> (Pose) {
    alphaBoolBlend {blendOption=Linear}
    boneToModify {boneName=thigh_l_fix_a}
    constraintSetup [{targetBone {boneName=thigh_twist_01_l}}, {targetBone {boneName=pelvis}}]
    constraintWeights [1,1]
   }
   n10 = "Constraint - thigh_r_fix_a (PP)"!(ComponentPose=@n9, Alpha=1, ConstraintWeights_0=0.5, ConstraintWeights_1:float=0.5) -> (Pose) {
    alphaBoolBlend {blendOption=Linear}
    boneToModify {boneName=thigh_r_fix_a}
    constraintSetup [{targetBone {boneName=thigh_twist_01_r}}, {targetBone {boneName=pelvis}}]
    constraintWeights [1,1]
   }
   n11 = "Constraint - dyn_Grenade_01_base (PP)"!(ComponentPose=@n10, Alpha=1, ConstraintWeights_0=0.5, ConstraintWeights_1:float=0.5) -> (Pose) {
    alphaBoolBlend {blendOption=Linear}
    boneToModify {boneName=dyn_Grenade_01_base}
    constraintSetup [{targetBone {boneName=spine_04}}, {targetBone {boneName=spine_05}}]
    constraintWeights [1,1]
   }
   n12 = Get leg_rot_l() -> (leg_rot_l:double)
   n13 = "Transform (Modify) Bone - Bone: thigh_l_fix_a"(ComponentPose=@n11, Alpha=@n12, Translation="5,20,0", Rotation="0,0,0", Scale="1,1,1") {
    rotationMode=BMM_Additive
    rotationSpace=BCS_BoneSpace
    scaleMode=BMM_Additive
    scaleSpace=BCS_BoneSpace
    translationMode=BMM_Additive
    translationSpace=BCS_BoneSpace
    alphaBoolBlend {blendOption=Linear}
    boneToModify {boneName=thigh_l_fix_a}
    translation=<5 20>
   }
   n14 = Get leg_rot_r() -> (leg_rot_r:double)
   n15 = "Transform (Modify) Bone - Bone: thigh_r_fix_a"(ComponentPose=@n13, Alpha=@n14, Translation="5,-25,0", Rotation="0,0,0", Scale="1,1,1") {
    rotationMode=BMM_Additive
    rotationSpace=BCS_BoneSpace
    scaleMode=BMM_Additive
    scaleSpace=BCS_BoneSpace
    translationMode=BMM_Additive
    translationSpace=BCS_BoneSpace
    alphaBoolBlend {blendOption=Linear}
    boneToModify {boneName=thigh_r_fix_a}
    translation=<5 -25>
   }
   n16 = "Transform (Modify) Bone - Bone: thigh_volume_bk_l"(ComponentPose=@n15, Alpha=@n12, Translation="20,0,0", Rotation="0,0,0", Scale="1,1,1") {
    rotationSpace=BCS_BoneSpace
    scaleMode=BMM_Additive
    scaleSpace=BCS_BoneSpace
    translationMode=BMM_Additive
    translationSpace=BCS_BoneSpace
    alphaBoolBlend {blendOption=Linear}
    boneToModify {boneName=thigh_volume_bk_l}
    translation {x=20}
   }
   n17 = "Transform (Modify) Bone - Bone: thigh_volume_bk_r"(ComponentPose=@n16, Alpha=@n14, Translation="2.860,5.043,0.000", Rotation="0,0,0", Scale="1,1,1") {
    rotationSpace=BCS_BoneSpace
    scaleMode=BMM_Additive
    scaleSpace=BCS_BoneSpace
    translationMode=BMM_Additive
    translationSpace=BCS_BoneSpace
    alphaBoolBlend {blendOption=Linear}
    boneToModify {boneName=thigh_volume_bk_r}
    translation=<2.86 5.043>
   }
   Save cached pose 'AnimCache'(Pose=Component To Local(ComponentPose=@n17)) {cacheName=AnimCache}
   Get Shoulder_rot_f_r() -> (Shoulder_rot_f_r:double)
  ]}
 EventGraph {nodes [
   Event BlueprintUpdateAnimation()
   n1 = Get Skel_Mesh()
   > Set leg_rot_r(leg_rot_r:double=GetAlphaFromReferenceTransformDelta(SkeletalMeshComponent=@n1, BoneName=thigh_r, BaseName=pelvis, InputRange=(X=0,Y=-150), TwistAxis="0,0,1")) -> (Output_Get:double)
   > Set leg_rot_l(leg_rot_l:double=GetAlphaFromReferenceTransformDelta(SkeletalMeshComponent=@n1, BoneName=thigh_twist_01_l, BaseName=pelvis, InputRange=(X=0,Y=-150), TwistAxis="0,0,1")) -> (Output_Get:double)
   n2 = Get Skel_Mesh()
   > Set Shoulder_rot_f_r(Shoulder_rot_f_r:double=GetAlphaFromReferenceTransformDelta(SkeletalMeshComponent=@n2, BoneName=upperarm_r, BaseName=clavicle_r, InputRange=(X=0,Y=90), TwistAxis="0,0,1")) -> (Output_Get:double)
   > Set Shoulder_rot_f_l(Shoulder_rot_f_l:double=GetAlphaFromReferenceTransformDelta(SkeletalMeshComponent=@n2, BoneName=upperarm_l, BaseName=clavicle_l, InputRange=(X=0,Y=60), TwistAxis="0,0,1")) -> (Output_Get:double)
   Event BlueprintInitializeAnimation()
   > Set Skel_Mesh(Skel_Mesh=GetOwnerSkeletalMeshComponent())
  ]}
]
variables [
 leg_rot_l {type=number}
 leg_rot_r {type=number}
 Shoulder_rot_f_l {type=number}
 Shoulder_rot_f_r {type=number}
 Skel_Mesh {type=SkeletalMeshComponent}
]
