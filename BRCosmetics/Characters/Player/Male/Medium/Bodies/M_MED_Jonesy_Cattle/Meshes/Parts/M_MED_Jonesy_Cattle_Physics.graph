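# Physics asset for a part of the male medium Jonesy Cattle player cosmetic, covering the head, a hant bone, two ears and left/right hant chains.
# A large capsule on hant is held within one degree of head, and four three-bone chains branch from it.
# Ear joints share uniform moderate swing and twist limits, while hant side chains are looser at the base and stiffer toward the tips.
# Only hant and outer ear bones have sized capsules; the rest are tiny placeholders.
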PhysicsAsset /BRCosmetics/Characters/Player/Male/Medium/Bodies/M_MED_Jonesy_Cattle/Meshes/Parts/M_MED_Jonesy_Cattle_Physics
bodies {
 dny_ear_l_01 [{shapeType=Capsule, length=2.799, radius=2.5331, center=<-0.3882 0.0529 -0.9526>, rotation=<rot 0 86.0457 -78.0948>}]
 dny_ear_l_02 [{shapeType=Capsule, length=3.0736, radius=2.1429, center=<0.0296 0.1223 -1.0362>, rotation=<rot 2.5567 -89.9995 -107.8483>}]
 dny_ear_l_base [{shapeType=Capsule, length=0.505, radius=0.505, rotation=<rot 0 -90 0>}]
 dny_ear_r_01 [{shapeType=Capsule, length=3.6371, radius=2.0416, center=<0.0073 -0.0314 0.8622>, rotation=<rot 0 85.9903 -69.775>}]
 dny_ear_r_02 [{shapeType=Capsule, length=4.6437, radius=1.981, center=<0.3617 -0.1353 0.7414>, rotation=<rot 2.5617 -89.9999 -113.0931>}]
 dny_ear_r_base [{shapeType=Capsule, length=0.505, radius=0.505, rotation=<rot 0 -90 0>}]
 hant [{shapeType=Capsule, length=17.7618, radius=12.8955, center=<7.7748 0.1024 0.0373>, rotation=<rot -0.3247 -90 -0.0879>}]
 hant_l_01 [{shapeType=Capsule, length=0.505, radius=0.505, rotation=<rot 0 -90 0>}]
 hant_l_02 [{shapeType=Capsule, length=0.505, radius=0.505, rotation=<rot 0 -90 0>}]
 hant_l_base [{shapeType=Capsule, length=0.505, radius=0.505, rotation=<rot 0 -90 0>}]
 hant_r_01 [{shapeType=Capsule, length=0.505, radius=0.505, rotation=<rot 0 -90 0>}]
 hant_r_02 [{shapeType=Capsule, length=0.505, radius=0.505, rotation=<rot 0 -90 0>}]
 hant_r_base [{shapeType=Capsule, length=0.505, radius=0.505, rotation=<rot 0 -90 0>}]
 head [{shapeType=Capsule, length=0.505, radius=0.505, rotation=<rot 0 -90 0>}]
}
constraints [
 {bone1Name=hant, bone2Name=head, swing1LimitDegrees=1, swing2LimitDegrees=1, twistLimitDegrees=1}
 {bone1Name=dny_ear_l_base, bone2Name=hant, swing1LimitDegrees=25, swing2LimitDegrees=25, twistLimitDegrees=25}
 {bone1Name=dny_ear_l_01, bone2Name=dny_ear_l_base, swing1LimitDegrees=25, swing2LimitDegrees=25, twistLimitDegrees=25}
 {bone1Name=dny_ear_l_02, bone2Name=dny_ear_l_01, swing1LimitDegrees=25, swing2LimitDegrees=25, twistLimitDegrees=25}
 {bone1Name=dny_ear_r_base, bone2Name=hant, swing1LimitDegrees=25, swing2LimitDegrees=25, twistLimitDegrees=25}
 {bone1Name=dny_ear_r_01, bone2Name=dny_ear_r_base, swing1LimitDegrees=25, swing2LimitDegrees=25, twistLimitDegrees=25}
 {bone1Name=dny_ear_r_02, bone2Name=dny_ear_r_01, swing1LimitDegrees=25, swing2LimitDegrees=25, twistLimitDegrees=25}
 {bone1Name=hant_l_base, bone2Name=hant, swing1LimitDegrees=25, swing2LimitDegrees=25, twistLimitDegrees=35}
 {bone1Name=hant_l_01, bone2Name=hant_l_base, swing1LimitDegrees=7, swing2LimitDegrees=7, twistLimitDegrees=17}
 {bone1Name=hant_l_02, bone2Name=hant_l_01, swing1LimitDegrees=7, swing2LimitDegrees=7, twistLimitDegrees=17}
 {bone1Name=hant_r_base, bone2Name=hant, swing1LimitDegrees=25, swing2LimitDegrees=25, twistLimitDegrees=35}
 {bone1Name=hant_r_01, bone2Name=hant_r_base, swing1LimitDegrees=7, swing2LimitDegrees=7, twistLimitDegrees=17}
 {bone1Name=hant_r_02, bone2Name=hant_r_01, swing1LimitDegrees=7, swing2LimitDegrees=7, twistLimitDegrees=17}
]
