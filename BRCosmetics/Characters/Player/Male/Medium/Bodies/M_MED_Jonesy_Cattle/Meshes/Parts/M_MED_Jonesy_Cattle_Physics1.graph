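# Physics asset for the Jonesy Cattle male body cosmetic, defining capsule bodies and joint limits for the head, a central hant bone and its ear and side chains.
# The large hant capsule is held almost rigid to head and anchors four three-bone chains: left and right dny_ear and left and right hant.
# Ear joints have tight limits, stiffer on the right, while hant side chains swing more freely on small uniform capsules.
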PhysicsAsset /BRCosmetics/Characters/Player/Male/Medium/Bodies/M_MED_Jonesy_Cattle/Meshes/Parts/M_MED_Jonesy_Cattle_Physics1
bodies {
 dny_ear_l_01 [{shapeType=Capsule, length=2.404, radius=1.0082, center=<0.1623 0.1584 -1.4384>, rotation=<rot -3.9857 83.2965 -179.7612>}]
 dny_ear_l_02 [{shapeType=Capsule, length=1.9788, radius=0.7777, center=<0.1436 0.0866 -0.4913>, rotation=<rot -3.9279 -89.1489 169.7261>}]
 dny_ear_l_base [{shapeType=Capsule, length=0.3438, radius=1.3526, rotation=<rot 0 -90 0>}]
 dny_ear_r_01 [{shapeType=Capsule, length=3.1461, radius=1.1033, center=<-0.0623 0.0476 1.1952>, rotation=<rot -0.2221 82.4186 177.6624>}]
 dny_ear_r_02 [{shapeType=Capsule, length=2.0451, radius=0.8755, center=<0.0408 0.229 0.8739>, rotation=<rot 2.7511 -88.3336 178.6143>}]
 dny_ear_r_base [{shapeType=Capsule, length=0.3438, radius=1.4682, rotation=<rot 0 -90 0>}]
 hant [{shapeType=Capsule, length=6.4557, radius=7.8372, center=<2.5624 0.1024 0.0293>, rotation=<rot -0.3247 -90 -0.0879>}]
 hant_l_01 [{shapeType=Capsule, length=0.505, radius=0.505, rotation=<rot 0 -90 0>}]
 hant_l_02 [{shapeType=Capsule, length=0.505, radius=0.505, rotation=<rot 0 -90 0>}]
 hant_l_base [{shapeType=Capsule, length=0.505, radius=0.505, rotation=<rot 0 -90 0>}]
 hant_r_01 [{shapeType=Capsule, length=0.505, radius=0.505, rotation=<rot 0 -90 0>}]
 hant_r_02 [{shapeType=Capsule, length=0.505, radius=0.505, rotation=<rot 0 -90 0>}]
 hant_r_base [{shapeType=Capsule, length=0.505, radius=0.505, rotation=<rot 0 -90 0>}]
 head [{shapeType=Capsule, length=0.505, radius=0.505, rotation=<rot 0 -90 0>}]
}
constraints [
 {bone1Name=hant, bone2Name=head, swing1LimitDegrees=1, swing2LimitDegrees=1, twistLimitDegrees=1}
 {bone1Name=dny_ear_l_base, bone2Name=hant, swing1LimitDegrees=10, swing2LimitDegrees=10, twistLimitDegrees=10}
 {bone1Name=dny_ear_l_01, bone2Name=dny_ear_l_base, swing1LimitDegrees=10, swing2LimitDegrees=10, twistLimitDegrees=10}
 {bone1Name=dny_ear_l_02, bone2Name=dny_ear_l_01, swing1LimitDegrees=10, swing2LimitDegrees=10, twistLimitDegrees=10}
 {bone1Name=dny_ear_r_base, bone2Name=hant, swing1LimitDegrees=7, swing2LimitDegrees=7, twistLimitDegrees=7}
 {bone1Name=dny_ear_r_01, bone2Name=dny_ear_r_base, swing1LimitDegrees=7, swing2LimitDegrees=7, twistLimitDegrees=7}
 {bone1Name=dny_ear_r_02, bone2Name=dny_ear_r_01, swing1LimitDegrees=7, swing2LimitDegrees=7, twistLimitDegrees=7}
 {bone1Name=hant_l_base, bone2Name=hant, swing1LimitDegrees=25, swing2LimitDegrees=25, twistLimitDegrees=35}
 {bone1Name=hant_l_01, bone2Name=hant_l_base, swing1LimitDegrees=25, swing2LimitDegrees=25, twistLimitDegrees=35}
 {bone1Name=hant_l_02, bone2Name=hant_l_01, swing1LimitDegrees=25, swing2LimitDegrees=25, twistLimitDegrees=35}
 {bone1Name=hant_r_base, bone2Name=hant, swing1LimitDegrees=25, swing2LimitDegrees=25, twistLimitDegrees=35}
 {bone1Name=hant_r_01, bone2Name=hant_r_base, swing1LimitDegrees=25, swing2LimitDegrees=25, twistLimitDegrees=35}
 {bone1Name=hant_r_02, bone2Name=hant_r_01, swing1LimitDegrees=25, swing2LimitDegrees=25, twistLimitDegrees=35}
]
